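# Sound cue for the default impact sound of a squeaky (Sqweek) cosmetic pickaxe, stored in the TeslaCoil pickaxe folder.
# A Random node picks among five equally weighted squeak impact waves and feeds a Mixer that lowers the level, with the cue volume reduced further.
# Small melee weapon attenuation, a cosmetic pickaxe impact property sheet and the melee weapon sound class and concurrency set govern playback.
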SoundCue /BRCosmetics/Sounds/PickAxes/TeslaCoil/PickAxe_Impact_Default_Sqweek_Cue
root=Mixer_0
nodes [
 {id=Mixer_0, children [Random_0], properties {inputVolume [0.7]}}
 {id=Random_0, children [WavePlayer_4,WavePlayer_0,WavePlayer_1,WavePlayer_2,WavePlayer_3], properties {weights [1,1,1,1,1]}}
 {id=WavePlayer_0, properties {soundWaveAssetPtr=/BRCosmetics/Sounds/PickAxes/PA_Sqweek_Impact_04}}
 {id=WavePlayer_1, properties {soundWaveAssetPtr=/BRCosmetics/Sounds/PickAxes/PA_Sqweek_Impact_05}}
 {id=WavePlayer_2, properties {soundWaveAssetPtr=/BRCosmetics/Sounds/PickAxes/PA_Sqweek_Impact_01}}
 {id=WavePlayer_3, properties {soundWaveAssetPtr=/BRCosmetics/Sounds/PickAxes/PA_Sqweek_Impact_03}}
 {id=WavePlayer_4, properties {soundWaveAssetPtr=/BRCosmetics/Sounds/PickAxes/PA_Sqweek_Impact_02}}
]
properties {
 attenuationSettings=&Fort_Weapon_Melee_Sml_Att
 audioPropertiesSheet=/BRCosmetics/Sounds/Settings/PropertySheets/Pickaxe/APS_Cosmetic_Pickaxe_Impact
 maxDistance=2800
 soundClassObject=&Fort_WeaponMelee_PSM
 volumeMultiplier=0.5
 concurrencyOverrides {resolutionRule=PreventNew}
 concurrencySet [&Player_Melee_Weapon_Conc]
}
