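# Sound cue for the ready sound of the Tesla Coil pickaxe cosmetic, classed as melee weapon audio.
# A quality-level node picks among branches that each run a random wave choice through a modulator and close-range pickaxe attenuation.
# The first quality level draws from six waves, while the others draw from three of those to trim variation.
# A concurrency override prevents new instances once the limit is reached.
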SoundCue /BRCosmetics/Sounds/PickAxes/TeslaCoil/PickAxe_TeslaCoil_Ready_Cue
root=QualityLevel_0
nodes [
 {id=Attenuation_0, children [Modulator_2], properties {attenuationSettings=/Game/Sounds/Attenuation/Fort_Weapon_Pickaxe_Close_Att}}
 {id=Attenuation_1, children [Modulator_0], properties {attenuationSettings=&Fort_Weapon_Pickaxe_Close_Att}}
 {id=Attenuation_2, children [Modulator_1], properties {attenuationSettings=&Fort_Weapon_Pickaxe_Close_Att}}
 {id=Modulator_0, children [Random_1]}
 {id=Modulator_1, children [Random_0]}
 {id=Modulator_2, children [Random_2]}
 {id=QualityLevel_0, children [Attenuation_2, Attenuation_0, Attenuation_1, Attenuation_1]}
 {id=Random_0, children [WavePlayer_0,WavePlayer_1,WavePlayer_2,WavePlayer_3,WavePlayer_4,WavePlayer_5], properties {weights [1,1,1,1,1,1]}}
 {id=Random_1, children [WavePlayer_8,WavePlayer_6,WavePlayer_7], properties {weights [1,1,1]}}
 {id=Random_2, children [WavePlayer_18, WavePlayer_13, WavePlayer_16], properties {weights [1,1,1]}}
 {id=WavePlayer_0, properties {soundWaveAssetPtr=&/Weapon_TeslaCoil_PickAxe_Ready_24}}
 {id=WavePlayer_1, properties {soundWaveAssetPtr=&/Weapon_TeslaCoil_PickAxe_Ready_25}}
 {id=WavePlayer_13, properties {soundWaveAssetPtr=&/Weapon_TeslaCoil_PickAxe_Ready_28}}
 {id=WavePlayer_16, properties {soundWaveAssetPtr=&/Weapon_TeslaCoil_PickAxe_Ready_27}}
 {id=WavePlayer_18, properties {soundWaveAssetPtr=&/Weapon_TeslaCoil_PickAxe_Ready_29}}
 {id=WavePlayer_2, properties {soundWaveAssetPtr=&/Weapon_TeslaCoil_PickAxe_Ready_26}}
 {id=WavePlayer_3, properties {soundWaveAssetPtr=&Weapon_TeslaCoil_PickAxe_Ready_27}}
 {id=WavePlayer_4, properties {soundWaveAssetPtr=&Weapon_TeslaCoil_PickAxe_Ready_28}}
 {id=WavePlayer_5, properties {soundWaveAssetPtr=&Weapon_TeslaCoil_PickAxe_Ready_29}}
 {id=WavePlayer_6, properties {soundWaveAssetPtr=&Weapon_TeslaCoil_PickAxe_Ready_28}}
 {id=WavePlayer_7, properties {soundWaveAssetPtr=&Weapon_TeslaCoil_PickAxe_Ready_27}}
 {id=WavePlayer_8, properties {soundWaveAssetPtr=&Weapon_TeslaCoil_PickAxe_Ready_29}}
]
properties {
 audioPropertiesSheet=/BRCosmetics/Sounds/Settings/PropertySheets/Pickaxe/APS_Cosmetic_Pickaxe_Ready
 bOverrideConcurrency=true
 maxDistance=1536
 soundClassObject=&Fort_WeaponMelee
 volumeMultiplier=0.12
 concurrencyOverrides {resolutionRule=PreventNew}
}
