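# Sound cue for the Tesla Coil pickaxe swing in the Battle Royale cosmetics audio.
# Each branch mixes close swing waves with distant swing waves, each layer using its own close or medium pickaxe attenuation, then passes through a modulator.
# A quality-level root scales variety from seven random waves per layer to three, then to a single wave shared by the last two levels.
# A concurrency override prevents new instances once the limit is reached.
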SoundCue /BRCosmetics/Sounds/PickAxes/TeslaCoil/PickAxe_TeslaCoil_Swing_Cue
root=QualityLevel_1
nodes [
 {id=Attenuation_0, children [Random_5], properties {attenuationSettings=/Game/Sounds/Attenuation/Fort_Weapon_Pickaxe_Medium_Att}}
 {id=Attenuation_1, children [WavePlayer_4], properties {attenuationSettings=/Game/Sounds/Attenuation/Fort_Weapon_Pickaxe_Close_Att}}
 {id=Attenuation_2, children [Random_4], properties {attenuationSettings=&Fort_Weapon_Pickaxe_Close_Att}}
 {id=Attenuation_3, children [WavePlayer_2], properties {attenuationSettings=&Fort_Weapon_Pickaxe_Medium_Att}}
 {id=Attenuation_4, children [Random_6], properties {attenuationSettings=&Fort_Weapon_Pickaxe_Close_Att}}
 {id=Attenuation_5, children [Random_7], properties {attenuationSettings=&Fort_Weapon_Pickaxe_Medium_Att}}
 {id=Mixer_0, children [Attenuation_2, Attenuation_0], properties {inputVolume [1,1]}}
 {id=Mixer_1, children [Attenuation_1, Attenuation_3], properties {inputVolume [1,1]}}
 {id=Mixer_2, children [Attenuation_4, Attenuation_5], properties {inputVolume [1,1]}}
 {id=Modulator_0, children [Mixer_0]}
 {id=Modulator_1, children [Mixer_1]}
 {id=Modulator_3, children [Mixer_2]}
 {id=QualityLevel_1, children [Modulator_0,Modulator_3,Modulator_1,Modulator_1]}
 {
  id=Random_4
  children [WavePlayer_6, WavePlayer_7, WavePlayer_8, WavePlayer_9, WavePlayer_10, WavePlayer_11, WavePlayer_12]
  properties {weights [1,1,1,1,1,1,1]}
 }
 {
  id=Random_5
  children [WavePlayer_23, WavePlayer_24, WavePlayer_25, WavePlayer_26, WavePlayer_27, WavePlayer_28, WavePlayer_29]
  properties {weights [1,1,1,1,1,1,1]}
 }
 {id=Random_6, children [WavePlayer_36, WavePlayer_35, WavePlayer_34], properties {weights [1,1,1]}}
 {id=Random_7, children [WavePlayer_32, WavePlayer_30, WavePlayer_31], properties {weights [1,1,1]}}
 {id=WavePlayer_10, properties {soundWaveAssetPtr=&/Weapon_TeslaCoil_PickAxe_Swings_14}}
 {id=WavePlayer_11, properties {soundWaveAssetPtr=&/Weapon_TeslaCoil_PickAxe_Swings_15}}
 {id=WavePlayer_12, properties {soundWaveAssetPtr=&/Weapon_TeslaCoil_PickAxe_Swings_16}}
 {id=WavePlayer_2, properties {soundWaveAssetPtr=&/Weapon_TeslaCoil_PickAxe_Swings_Distant-008}}
 {id=WavePlayer_23, properties {soundWaveAssetPtr=&/Weapon_TeslaCoil_PickAxe_Swings_Distant-001}}
 {id=WavePlayer_24, properties {soundWaveAssetPtr=&/Weapon_TeslaCoil_PickAxe_Swings_Distant-002}}
 {id=WavePlayer_25, properties {soundWaveAssetPtr=&Weapon_TeslaCoil_PickAxe_Swings_Distant-008}}
 {id=WavePlayer_26, properties {soundWaveAssetPtr=&/Weapon_TeslaCoil_PickAxe_Swings_Distant-009}}
 {id=WavePlayer_27, properties {soundWaveAssetPtr=&/Weapon_TeslaCoil_PickAxe_Swings_Distant-010}}
 {id=WavePlayer_28, properties {soundWaveAssetPtr=&/Weapon_TeslaCoil_PickAxe_Swings_Distant-011}}
 {id=WavePlayer_29, properties {soundWaveAssetPtr=&/Weapon_TeslaCoil_PickAxe_Swings_Distant-012}}
 {id=WavePlayer_30, properties {soundWaveAssetPtr=&Weapon_TeslaCoil_PickAxe_Swings_Distant-010}}
 {id=WavePlayer_31, properties {soundWaveAssetPtr=&Weapon_TeslaCoil_PickAxe_Swings_Distant-011}}
 {id=WavePlayer_32, properties {soundWaveAssetPtr=&Weapon_TeslaCoil_PickAxe_Swings_Distant-008}}
 {id=WavePlayer_34, properties {soundWaveAssetPtr=&Weapon_TeslaCoil_PickAxe_Swings_15}}
 {id=WavePlayer_35, properties {soundWaveAssetPtr=&/Weapon_TeslaCoil_PickAxe_Swings_13}}
 {id=WavePlayer_36, properties {soundWaveAssetPtr=&/Weapon_TeslaCoil_PickAxe_Swings_12}}
 {id=WavePlayer_4, properties {soundWaveAssetPtr=&Weapon_TeslaCoil_PickAxe_Swings_13}}
 {id=WavePlayer_6, properties {soundWaveAssetPtr=&/Weapon_TeslaCoil_PickAxe_Swings_10}}
 {id=WavePlayer_7, properties {soundWaveAssetPtr=&/Weapon_TeslaCoil_PickAxe_Swings_11}}
 {id=WavePlayer_8, properties {soundWaveAssetPtr=&Weapon_TeslaCoil_PickAxe_Swings_12}}
 {id=WavePlayer_9, properties {soundWaveAssetPtr=&Weapon_TeslaCoil_PickAxe_Swings_13}}
]
properties {
 audioPropertiesSheet=/BRCosmetics/Sounds/Settings/PropertySheets/Pickaxe/APS_Cosmetic_Pickaxe_Swing
 bOverrideConcurrency=true
 maxDistance=3000
 soundClassObject=&Fort_WeaponMelee_PSM
 volumeMultiplier=0.12
 concurrencyOverrides {resolutionRule=PreventNew}
}
